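# Sound cue for the activation of a Mythic Labors time trial quest, routed through the Fort_Quests sound class.
# A fixed one-second delay feeds a single wave player holding the quest activate sound.
# The cue uses the ring-activated attenuation asset, boosts volume, and overrides concurrency so new instances are refused once the limit is reached.
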SoundCue /Game/Athena/Sounds/Quests/TimedTrial/New/Cue_MythicLabors_TimeTrial_Activate_Cue
root=Delay_0
nodes [
 {id=Delay_0, children [WavePlayer_1], properties {delayMax=1, delayMin=1}}
 {id=WavePlayer_1, properties {soundWaveAssetPtr=&/MythicLabors_Quest_Activate}}
]
properties {
 attenuationSettings=&/Att_MythicLabors_Ring_Activated
 bOverrideConcurrency=true
 maxDistance=2097152
 soundClassObject=&Fort_Quests
 volumeMultiplier=1.5
 concurrencyOverrides {resolutionRule=PreventNew}
}
