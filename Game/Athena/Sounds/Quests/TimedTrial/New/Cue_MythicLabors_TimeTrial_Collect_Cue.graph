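# Sound cue for collecting in the Mythic Labors time trial quest, using checkpoint-activated sounds.
# A Random node picks one of three equally weighted wave variations and is excluded from branch culling.
# The cue plays in the quest sound class at boosted volume, with attenuation overrides that turn off distance falloff and spatialization.
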SoundCue /Game/Athena/Sounds/Quests/TimedTrial/New/Cue_MythicLabors_TimeTrial_Collect_Cue
root=Random_0
nodes [
 {id=Random_0, children [WavePlayer_2,WavePlayer_3,WavePlayer_4], properties {bShouldExcludeFromBranchCulling=true, weights [1,1,1]}}
 {id=WavePlayer_2, properties {soundWaveAssetPtr=&/MythicLabors_Checkpoint_Activated_03}}
 {id=WavePlayer_3, properties {soundWaveAssetPtr=&/MythicLabors_Checkpoint_Activated_02}}
 {id=WavePlayer_4, properties {soundWaveAssetPtr=&/MythicLabors_Checkpoint_Activated_01}}
]
properties {
 maxDistance=2097152
 priority=2
 soundClassObject=&Fort_Quests
 volumeMultiplier=1.5
 attenuationOverrides {bAttenuate=false, bSpatialize=false}
}
