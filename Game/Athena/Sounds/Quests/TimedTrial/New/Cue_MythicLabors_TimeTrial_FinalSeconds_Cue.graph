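# Sound cue for the final-seconds countdown in the Mythic Labors time trial quest.
# It plays a single wave player with the MythicLabors_FinalSeconds wave, routed through the Fort_Quests sound class with boosted volume and a very large max distance.
# A concurrency override blocks new instances while one is already playing, so the countdown does not stack.
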SoundCue /Game/Athena/Sounds/Quests/TimedTrial/New/Cue_MythicLabors_TimeTrial_FinalSeconds_Cue
root=WavePlayer_1
nodes [{id=WavePlayer_1, properties {soundWaveAssetPtr=&/MythicLabors_FinalSeconds}}]
properties {
 bOverrideConcurrency=true
 maxDistance=2097152
 soundClassObject=&Fort_Quests
 volumeMultiplier=1.5
 concurrencyOverrides {resolutionRule=PreventNew}
}
