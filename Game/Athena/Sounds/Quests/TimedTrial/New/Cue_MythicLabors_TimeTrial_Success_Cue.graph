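# Sound cue for the success of a Mythic Labors timed trial quest in Athena.
# It plays a single wave, the Mythic Labors quest complete sound, through one wave player with no randomization or mixing.
# The cue routes to the quests sound class with a slight volume boost, and its attenuation override turns off distance falloff and spatialization.
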SoundCue /Game/Athena/Sounds/Quests/TimedTrial/New/Cue_MythicLabors_TimeTrial_Success_Cue
root=WavePlayer_0
nodes [{id=WavePlayer_0, properties {soundWaveAssetPtr=&/MythicLabors_Quest_Complete}}]
properties {
 maxDistance=2097152
 soundClassObject=&Fort_Quests
 volumeMultiplier=1.2
 attenuationOverrides {bAttenuate=false, bSpatialize=false}
}
