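# Sound cue that plays the music for the Mythic Labors time trial quest.
# A fixed short delay precedes a single wave player for the MythicLabors_TimeTrial_Music wave.
# The cue is routed to the Fort_Quests sound class.
# It stores attenuation overrides and a PreventNew concurrency rule, with no enable flag shown for either.
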SoundCue /Game/Athena/Sounds/Quests/TimedTrial/New/Cue_Mythic_Labors_TimeTrial_Music
root=Delay_0
nodes [
 {id=Delay_0, children [WavePlayer_1], properties {delayMax=0.65, delayMin=0.65}}
 {id=WavePlayer_1, properties {soundWaveAssetPtr=&/MythicLabors_TimeTrial_Music}}
]
properties {
 maxDistance=2097152
 soundClassObject=&Fort_Quests
 volumeMultiplier=1
 attenuationOverrides {falloffDistance=1000, attenuationShapeExtents {x=200}}
 concurrencyOverrides {resolutionRule=PreventNew}
}
